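# Light function material for a pulsing colored light, stored with the weapon skin wrap materials.
# It multiplies an orange-default "color light" vector parameter by the MF_Pulse_Illuminate function, whose brightness, frequency, sine-wave and clamp inputs are left unwired.
# A lobby caustics texture sample sits unconnected, and the file shows no link from the multiply to a material output.
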Material /Game/Weapons/WeaponSkins/Wraps/Materials/M_pulse_illuminate
graph {
 properties {materialDomain=MD_LightFunction, pixelDepthOffsetMode=PDOM_Legacy, usedShadingModels=Default Lit}
 nodes [
  Multiply(
   A=VectorParameter() {parameterName=color light, defaultValue=<rgba 1 0.3445 0.1129 1>}.RGB
   B=MaterialFunctionCall!(Brightness-Low, Brightness-High, Frequency, Sinewave-Offset, Amplitude, Clamp Sine Wave?, Clamp-Min, Clamp-Max) {materialFunction=/Game/Weapons/WeaponSkins/Wraps/MaterialFunctions/MF_Pulse_Illuminate}.Result
  )
  TextureSample() {texture=/Game/Environments/FrontEnd/BackPlates/Textures/T_Lobby_Caustics2}
 ]
}
